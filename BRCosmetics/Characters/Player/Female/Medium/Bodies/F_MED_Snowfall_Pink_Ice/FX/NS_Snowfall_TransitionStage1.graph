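# Niagara effect for the first transition stage of the Snowfall (Pink Ice) character body cosmetic, emitting icy blue sprites from a skeletal mesh.
# All three emitters use FN_BoneFlow, mapping spawn groups to bones with per-group height ranges.
# Decal sprites re-evaluate bone placement every update.
# SnowFlakes combines bursts and decaying spawn rates with curl noise, while cold_smoke bursts drifting, animated puffs that partly follow source movement.
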NiagaraSystem /BRCosmetics/Characters/Player/Female/Medium/Bodies/F_MED_Snowfall_Pink_Ice/FX/NS_Snowfall_TransitionStage1
dependencies {
 usedDataInterfaces [&NiagaraDataInterfaceArrayFloat, &NiagaraDataInterfaceCurve]
 usedDynamicInputs []
 usedModules [&SystemState]
}
emitters {
 Decal {
  emitterUpdateScript [
   &EmitterState()
   &SpawnRate(SpawnGroup=1, SpawnRate=90, Use Spawn Probability=false) {enabled=false}
   &SpawnBurst_Instantaneous(Use Spawn Probability=false, UseLoopCountLimit=false) {module=SpawnBurst_Instantaneous004}
   &SpawnBurst_Instantaneous(Spawn Group=1, Use Spawn Probability=false, UseLoopCountLimit=false) {module=SpawnBurst_Instantaneous003}
   &SpawnBurst_Instantaneous(Spawn Group=2, Use Spawn Probability=false, UseLoopCountLimit=false) {module=SpawnBurst_Instantaneous002}
   &SpawnBurst_Instantaneous(Spawn Group=3, Use Spawn Probability=false, UseLoopCountLimit=false) {module=SpawnBurst_Instantaneous001}
   &SpawnBurst_Instantaneous(Spawn Count=2, Spawn Group=4, Use Spawn Probability=false, UseLoopCountLimit=false)
   &SpawnBurst_Instantaneous(Spawn Count=6, Spawn Group=5, Use Spawn Probability=false, UseLoopCountLimit=false) {enabled=false, module=SpawnBurst_Instantaneous005}
  ]
  particleSpawnScript [
   &InitializeParticle(
    Color=<rgba 0.1173 0.7685 0.9387 1>
    Lifetime Max=2.5
    Lifetime Min=3
    Lifetime Mode=::NewEnumerator1
    Mass Mode=::NewEnumerator3
    Sprite Rotation Mode=::NewEnumerator2
    Sprite Size Mode=::NewEnumerator1
    Uniform Sprite Size Max=50
    Uniform Sprite Size Min=50
    UsePositionOffset=false
    Write Material Random=true
   )
   &AddVelocityInCone(
    Cone Angle=20
    Cone Axis=<0 0 1>
    Module.Use Velocity Falloff On Cone Axis=false
    Velocity Distribution Along Cone Axis=0.75
    Velocity Strength=&RandomRangeFloat(Minimum=500, Maximum=850)
   ) {enabled=false}
   /SetVariables(Particles.SpawnGroup=&RandomRangeInt(Minimum=0, Maximum=5)) {enabled=false, isSetParameters=true, module=SetVariables_#1}
   &InheritVelocity(Inherited Velocity Amount Scale=<0.75 0.75 0.75>, Speed Limit=true, UseSpeedThreshold=false) {enabled=false}
   &AddVelocity(Velocity=&Multiply_Vector(
     A=&RotateVectorByQuaternion(
      Vector To Rotate=$Output.FN_BoneFlow.SplineNormal
      Quaternion=&MakeQuatFromAxisAngle(
       Axis=$Output.FN_BoneFlow.SplineTangent
       AngleInDegrees=&RandomRangeFloat(
        Minimum=&SelectFloatFromArray(
         Direct Array Index=&SpawnGroup()
         Float Selection Array=DataInterface() {propertyValues {FloatData [0,0,100,0,-45]}}
        )
        Maximum=&SelectFloatFromArray(
         Direct Array Index=&SpawnGroup()
         Float Selection Array=DataInterface() {propertyValues {FloatData [360,360,140,360,-35]}}
        )
       )
      )
     )
     B=&VectorFromFloat(Value=75)
    )) {enabled=false}
   &CameraOffset(Camera Offset Amount=25)
   /SetVariables(Particles.rand int:NiagaraInt32=&RandomRangeInt(Minimum=0, Maximum=63)) {isSetParameters=true, module=SetVariables_#2}
   &DynamicMaterialParameters(
    Index 0 Param 1=0.5
    Index 0 Param 2=1
    Index 0 Param 4=1
    Module.Param0WriteEnabled=true
    Module.Param0WriteEnabled1=true
    Module.Param0WriteEnabled2=true
    Module.Param0WriteEnabled3=true
    Module.Param1WriteEnabled=true
    Module.Param1WriteEnabled1=true
    Module.Param1WriteEnabled2=true
    Module.Param1WriteEnabled3=true
    Module.Param2WriteEnabled=true
    Module.Param2WriteEnabled1=true
    Module.Param2WriteEnabled2=true
    Module.Param2WriteEnabled3=true
    Module.Param3WriteEnabled=true
    Module.Param3WriteEnabled1=true
    Module.Param3WriteEnabled2=true
    Module.Param3WriteEnabled3=true
    Write Parameter Index 1=true
    Write Parameter Index 2=true
    Write Parameter Index 3=true
   )
   /SetVariables() {isSetParameters=true}
  ]
  particleUpdateScript [
   &ParticleState()
   &FN_BoneFlow(
    Angle:NiagaraFloat=&RandomRangeFloat(Minimum=0, Maximum=1)
    BoneID:NiagaraInt32=&SpawnGroup()
    Height:NiagaraFloat=&RandomRangeFloat(
     Minimum=&SelectFloatFromArray(
      Direct Array Index=&SpawnGroup()
      Float Selection Array=DataInterface() {propertyValues {FloatData [0.75,-0.1,1,0,0,0]}}
     )
     Maximum=&SelectFloatFromArray(
      Direct Array Index=&SpawnGroup()
      Float Selection Array=DataInterface() {propertyValues {FloatData [1.1,0.1,0.8,0.15,1,0]}}
     )
    )
    Radius=1
    Set Material Dynamic=false
    SkeletalMesh:NiagaraDataInterfaceSkeletalMesh=$User.SkeletalMesh
   )
   &ScaleColor(
    Scale Alpha=&FloatFromCurve(FloatCurve=DataInterface() {propertyValues {Curve {keys [
         {arriveTangent=3.8574, interpMode=RCIM_Cubic, leaveTangent=3.8574, tangentMode=RCTM_User, time=0, value=0}
         {
          arriveTangent=0.1363
          arriveTangentWeight=0.8003
          interpMode=RCIM_Cubic
          leaveTangent=-5.0014
          leaveTangentWeight=0.8003
          tangentMode=RCTM_Break
          time=0.5
          value=1
         }
         {tangentMode=RCTM_User, time=1, value=0}
        ]}}})
    ScaleA=true
    ScaleRGB=true
   )
   &ScaleSpriteSize(Uniform Curve Sprite Scale=DataInterface() {propertyValues {Curve {keys [
        <at 0 0.5 RCIM_Cubic>
        <at 0.65 1 RCIM_Cubic>
        {arriveTangent=-2.0287, interpMode=RCIM_Cubic, leaveTangent=-2.0287, tangentMode=RCTM_User, time=1, value=0.5}
       ]}}})
  ]
  rendererClasses [&NiagaraSpriteRendererProperties]
  renderers [
   {rendererClass=&NiagaraSpriteRendererProperties}
   {rendererClass=&NiagaraSpriteRendererProperties, rendererIndex=1}
  ]
 }
 SnowFlakes {
  emitterSpawnScript [/SetVariables(Emitter.Flake Rate=DataInterface() {propertyValues {Curve {keys [
        {
         arriveTangent=-2.5851
         arriveTangentWeight=0.8398
         interpMode=RCIM_Cubic
         leaveTangent=-2.5851
         leaveTangentWeight=0.8398
         tangentMode=RCTM_User
         tangentWeightMode=RCTWM_WeightedBoth
         time=0
         value=1
        }
        {
         arriveTangentWeight=1.0234
         interpMode=RCIM_Cubic
         leaveTangentWeight=1.0234
         tangentMode=RCTM_User
         tangentWeightMode=RCTWM_WeightedBoth
         time=1
         value=0
        }
       ]}}}) {isSetParameters=true}]
  emitterUpdateScript [
   &EmitterState(Life Cycle Mode=::NewEnumerator1, Loop Behavior=::NewEnumerator1, Loop Duration=3, UseLoopDelay=false)
   &SpawnRate(SpawnGroup=1, SpawnRate=90, Use Spawn Probability=false) {enabled=false}
   &SpawnBurst_Instantaneous(Spawn Count=5, Use Spawn Probability=false, UseLoopCountLimit=false) {module=SpawnBurst_Instantaneous004}
   &SpawnBurst_Instantaneous(Spawn Count=5, Spawn Group=1, Use Spawn Probability=false, UseLoopCountLimit=false) {module=SpawnBurst_Instantaneous003}
   &SpawnBurst_Instantaneous(Spawn Count=5, Spawn Group=2, Use Spawn Probability=false, UseLoopCountLimit=false) {module=SpawnBurst_Instantaneous002}
   &SpawnBurst_Instantaneous(Spawn Count=5, Spawn Group=3, Use Spawn Probability=false, UseLoopCountLimit=false) {module=SpawnBurst_Instantaneous001}
   &SpawnBurst_Instantaneous(Spawn Count=6, Spawn Group=4, Use Spawn Probability=false, UseLoopCountLimit=false)
   &SpawnBurst_Instantaneous(Spawn Count=6, Spawn Group=5, Use Spawn Probability=false, UseLoopCountLimit=false) {enabled=false, module=SpawnBurst_Instantaneous005}
   &SpawnRate(
    SpawnRate=&FloatFromCurve(FloatCurve=$Emitter.Flake Rate, CurveIndex=$Emitter.NormalizedLoopAge, Scale Curve=10)
    Use Spawn Probability=false
   ) {module=SpawnRate001}
   &SpawnRate(
    SpawnGroup=1
    SpawnRate=&FloatFromCurve(FloatCurve=$Emitter.Flake Rate, CurveIndex=$Emitter.NormalizedLoopAge, Scale Curve=10)
    Use Spawn Probability=false
   ) {module=SpawnRate006}
   &SpawnRate(
    SpawnGroup=2
    SpawnRate=&FloatFromCurve(FloatCurve=$Emitter.Flake Rate, CurveIndex=$Emitter.NormalizedLoopAge, Scale Curve=10)
    Use Spawn Probability=false
   ) {module=SpawnRate005}
   &SpawnRate(
    SpawnGroup=3
    SpawnRate=&FloatFromCurve(FloatCurve=$Emitter.Flake Rate, CurveIndex=$Emitter.NormalizedLoopAge, Scale Curve=10)
    Use Spawn Probability=false
   ) {module=SpawnRate004}
   &SpawnRate(
    SpawnGroup=4
    SpawnRate=&FloatFromCurve(FloatCurve=$Emitter.Flake Rate, CurveIndex=$Emitter.NormalizedLoopAge, Scale Curve=10)
    Use Spawn Probability=false
   ) {module=SpawnRate003}
   &SpawnBurst_Instantaneous(Spawn Count=6, Spawn Time=0.6342, Use Spawn Probability=false, UseLoopCountLimit=false) {module=SpawnBurst_Instantaneous006}
  ]
  particleSpawnScript [
   &InitializeParticle(
    Color=<rgba 0.261 0.7983 0.9387 0.125>
    Lifetime Max=1
    Lifetime Min=0.75
    Lifetime Mode=::NewEnumerator1
    Mass Mode=::NewEnumerator3
    Sprite Rotation Mode=::NewEnumerator2
    Sprite Size Mode=::NewEnumerator1
    Uniform Sprite Size Max=7
    UsePositionOffset=false
    Write Material Random=true
   )
   &AddVelocityInCone(
    Cone Angle=20
    Cone Axis=<0 0 1>
    Module.Use Velocity Falloff On Cone Axis=false
    Velocity Distribution Along Cone Axis=0.75
    Velocity Strength=&RandomRangeFloat(Minimum=500, Maximum=850)
   ) {enabled=false}
   /SetVariables(Particles.SpawnGroup=&RandomRangeInt(Minimum=0, Maximum=5)) {enabled=false, isSetParameters=true, module=SetVariables_#1}
   &FN_BoneFlow(
    Angle:NiagaraFloat=&RandomRangeFloat(Minimum=0, Maximum=1)
    BoneID:NiagaraInt32=&SpawnGroup()
    Height:NiagaraFloat=&RandomRangeFloat(
     Minimum=&SelectFloatFromArray(
      Direct Array Index=&SpawnGroup()
      Float Selection Array=DataInterface() {propertyValues {FloatData [0.75,-0.1,1,0,0,0]}}
     )
     Maximum=&SelectFloatFromArray(
      Direct Array Index=&SpawnGroup()
      Float Selection Array=DataInterface() {propertyValues {FloatData [1.1,0.1,0.8,0.15,1,0]}}
     )
    )
    Radius=1
    Set Material Dynamic=false
    SkeletalMesh:NiagaraDataInterfaceSkeletalMesh=$User.SkeletalMesh
   )
   &InheritVelocity(Inherited Velocity Amount Scale=<0.75 0.75 0.75>, Speed Limit=true, UseSpeedThreshold=false)
   &AddVelocity(Velocity=&Multiply_Vector(
     A=&RotateVectorByQuaternion(
      Vector To Rotate=$Output.FN_BoneFlow.SplineNormal
      Quaternion=&MakeQuatFromAxisAngle(
       Axis=$Output.FN_BoneFlow.SplineTangent
       AngleInDegrees=&RandomRangeFloat(
        Minimum=&SelectFloatFromArray(
         Direct Array Index=&SpawnGroup()
         Float Selection Array=DataInterface() {propertyValues {FloatData [0,0,100,0,-45]}}
        )
        Maximum=&SelectFloatFromArray(
         Direct Array Index=&SpawnGroup()
         Float Selection Array=DataInterface() {propertyValues {FloatData [360,360,140,360,-35]}}
        )
       )
      )
     )
     B=&VectorFromFloat(Value=75)
    ))
   &CameraOffset(Camera Offset Amount=10)
   /SetVariables(Particles.rand int:NiagaraInt32=&RandomRangeInt(Minimum=0, Maximum=63)) {isSetParameters=true, module=SetVariables_#2}
   &SubUVAnimation(End Frame=3, SubUV Animation Mode:ENiagara_SubUVLookupMode=::NewEnumerator2)
   &DynamicMaterialParameters(
    Index 1 Param 3=1
    Index 1 Param 4=1
    Module.Param0WriteEnabled1=true
    Module.Param1WriteEnabled1=true
    Module.Param2WriteEnabled1=true
    Module.Param3WriteEnabled1=true
    Write Parameter Index 0=false
    Write Parameter Index 1=true
   )
  ]
  particleUpdateScript [
   &ParticleState()
   &GravityForce() {enabled=false}
   &Drag(Drag=3, Use Linear Drag=true, Use Rotational Drag=false)
   &ScaleColor(
    Scale Alpha=&FloatFromCurve(FloatCurve=DataInterface() {propertyValues {Curve {keys [
         {interpMode=RCIM_Cubic, tangentMode=RCTM_User, time=0, value=0}
         {
          arriveTangent=-4.6173
          arriveTangentWeight=0.8003
          interpMode=RCIM_Cubic
          leaveTangent=-4.6173
          leaveTangentWeight=0.8003
          tangentMode=RCTM_User
          time=0.5
          value=1
         }
         {tangentMode=RCTM_User, time=1, value=0}
        ]}}})
    ScaleA=true
    ScaleRGB=true
   )
   &CurlNoiseForce(Noise Strength=150, Pan Noise=false)
   &SolveForcesAndVelocity(Clamp Velocity=false, Limit Acceleration=false)
   &ScaleSpriteSize(Uniform Curve Sprite Scale=DataInterface() {propertyValues {Curve {keys [{tangentMode=RCTM_User, time=0, value=0.25}, {tangentMode=RCTM_User, time=1, value=1.25}]}}})
   &SpriteRotationRate(Rotation Rate=&RandomRangeFloat(Minimum=-120, Maximum=120))
  ]
  rendererClasses [&NiagaraSpriteRendererProperties]
  renderers [{rendererClass=&NiagaraSpriteRendererProperties}]
 }
 cold_smoke {
  emitterUpdateScript [
   &EmitterState(Life Cycle Mode=::NewEnumerator1, Loop Behavior=::NewEnumerator1, Loop Delay=1, Loop Duration=1, UseLoopDelay=true)
   &SpawnRate(SpawnGroup=1, SpawnRate=90, Use Spawn Probability=false) {enabled=false}
   &SpawnBurst_Instantaneous(Spawn Count=3, Use Spawn Probability=false, UseLoopCountLimit=false) {module=SpawnBurst_Instantaneous004}
   &SpawnBurst_Instantaneous(Spawn Count=3, Spawn Group=1, Use Spawn Probability=false, UseLoopCountLimit=false) {module=SpawnBurst_Instantaneous003}
   &SpawnBurst_Instantaneous(Spawn Count=3, Spawn Group=2, Use Spawn Probability=false, UseLoopCountLimit=false) {module=SpawnBurst_Instantaneous002}
   &SpawnBurst_Instantaneous(Spawn Count=3, Spawn Group=3, Use Spawn Probability=false, UseLoopCountLimit=false) {module=SpawnBurst_Instantaneous001}
   &SpawnBurst_Instantaneous(Spawn Count=6, Spawn Group=4, Use Spawn Probability=false, UseLoopCountLimit=false)
   &SpawnBurst_Instantaneous(Spawn Count=6, Spawn Group=5, Use Spawn Probability=false, UseLoopCountLimit=false) {enabled=false, module=SpawnBurst_Instantaneous005}
  ]
  particleSpawnScript [
   &InitializeParticle(
    Color=<rgba 0.261 0.7983 0.9387 0.2>
    Lifetime Mode=::NewEnumerator1
    Mass Mode=::NewEnumerator3
    Sprite Rotation Mode=::NewEnumerator2
    Sprite Size Mode=::NewEnumerator1
    Uniform Sprite Size Max=75
    Uniform Sprite Size Min=20
    UsePositionOffset=false
    Write Material Random=true
   )
   &AddVelocityInCone(
    Cone Angle=20
    Cone Axis=<0 0 1>
    Module.Use Velocity Falloff On Cone Axis=false
    Velocity Distribution Along Cone Axis=0.75
    Velocity Strength=&RandomRangeFloat(Minimum=500, Maximum=850)
   ) {enabled=false}
   /SetVariables(Particles.SpawnGroup=&RandomRangeInt(Minimum=0, Maximum=5)) {enabled=false, isSetParameters=true, module=SetVariables_#1}
   &FN_BoneFlow(
    Angle:NiagaraFloat=&RandomRangeFloat(Minimum=0, Maximum=1)
    BoneID:NiagaraInt32=&SpawnGroup()
    Height:NiagaraFloat=&RandomRangeFloat(
     Minimum=&SelectFloatFromArray(
      Direct Array Index=&SpawnGroup()
      Float Selection Array=DataInterface() {propertyValues {FloatData [0.75,-0.1,1,0,0,0]}}
     )
     Maximum=&SelectFloatFromArray(
      Direct Array Index=&SpawnGroup()
      Float Selection Array=DataInterface() {propertyValues {FloatData [1.1,0.1,0.8,0.15,1,0]}}
     )
    )
    Radius=1
    Set Material Dynamic=false
    SkeletalMesh:NiagaraDataInterfaceSkeletalMesh=$User.SkeletalMesh
   )
   &AddVelocity(Velocity=&Multiply_Vector(
     A=&RotateVectorByQuaternion(
      Vector To Rotate=$Output.FN_BoneFlow.SplineNormal
      Quaternion=&MakeQuatFromAxisAngle(
       Axis=$Output.FN_BoneFlow.SplineTangent
       AngleInDegrees=&RandomRangeFloat(
        Minimum=&SelectFloatFromArray(
         Direct Array Index=&SpawnGroup()
         Float Selection Array=DataInterface() {propertyValues {FloatData [0,0,100,0,-45]}}
        )
        Maximum=&SelectFloatFromArray(
         Direct Array Index=&SpawnGroup()
         Float Selection Array=DataInterface() {propertyValues {FloatData [360,360,140,360,-35]}}
        )
       )
      )
     )
     B=&VectorFromFloat(Value=25)
    ))
   &CameraOffset(Camera Offset Amount=15)
   /SetVariables(Particles.rand int:NiagaraInt32=&RandomRangeInt(Minimum=12, Maximum=20)) {isSetParameters=true, module=SetVariables_#2}
   &InheritVelocity(Inherited Velocity Amount Scale=<0.8 0.8 0.8>, Inherited Velocity Speed Limit=600, Speed Limit=false, UseSpeedThreshold=false) {enabled=false}
   &DynamicMaterialParameters(Module.Param0WriteEnabled=true, Module.Param1WriteEnabled=true, Module.Param2WriteEnabled=true, Module.Param3WriteEnabled=true)
  ]
  particleUpdateScript [
   &ParticleState()
   &GravityForce(Gravity=<0 0 -50>)
   &Drag(Use Linear Drag=true, Use Rotational Drag=false)
   &ScaleColor(
    Scale Alpha=&FloatFromCurve(FloatCurve=DataInterface() {propertyValues {Curve {keys [
         {interpMode=RCIM_Cubic, tangentMode=RCTM_User, time=0, value=0}
         {arriveTangentWeight=0.8003, interpMode=RCIM_Cubic, leaveTangentWeight=0.8003, time=0.5, value=1}
         {tangentMode=RCTM_User, time=1, value=0}
        ]}}})
    ScaleA=true
    ScaleRGB=true
   )
   &InheritSourceMovement(
    Applied Position Scale=&VectorFromFloat(Value=&FloatFromCurve(FloatCurve=DataInterface() {propertyValues {Curve {keys [{arriveTangent=-2.1026, interpMode=RCIM_Cubic, leaveTangent=-2.1026, tangentMode=RCTM_User, time=0.5, value=1}, <at 1 0>]}}}))
    ApplyForce=false
    ApplyPosition=true
    ApplyPositionFalloff=false
   )
   &SolveForcesAndVelocity(Clamp Velocity=false, Limit Acceleration=false)
   &ScaleSpriteSize(Uniform Curve Sprite Scale=DataInterface() {propertyValues {Curve {keys [{tangentMode=RCTM_User, time=0, value=0.25}, {tangentMode=RCTM_User, time=1, value=1.25}]}}})
   &SubUVAnimation(
    End Frame:NiagaraInt32=&Add_Integer(A=$Particles.Initial.rand int, B=63)
    Start Frame:NiagaraInt32=$Particles.Initial.rand int
    SubUV Loop Count=0.25
   )
  ]
  rendererClasses [&NiagaraSpriteRendererProperties]
  renderers [{rendererClass=&NiagaraSpriteRendererProperties}]
 }
}
summary {
 systemName=NS_Snowfall_TransitionStage1
 emitters [
  {emitterName=cold_smoke, rendererClasses [&NiagaraSpriteRendererProperties]}
  {emitterName=Decal, rendererClasses [&NiagaraSpriteRendererProperties]}
  {emitterName=SnowFlakes, rendererClasses [&NiagaraSpriteRendererProperties]}
 ]
 userVariables [User.SkeletalMesh {
   type=NiagaraDataInterfaceSkeletalMesh
   defaultValue {
    struct=NiagaraExt_VariableValue_DataInterface
    value {dataInterface=&self:NiagaraDataInterfaceSkeletalMesh_0}
   }
  }]
}
